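# Sound cue for the PoppySoap announcer's "Fight" voice line.
# A single wave player feeds the root directly, playing one announcer recording with no randomization or modulation.
# It uses announcer-specific attenuation and concurrency, a critical in-game UI sound class, boosted volume, and disabled virtualization.
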SoundCue /PoppySoap/Assets/Sound/IP_SC_PoppySoap_Announcer_VO_Fight
root=WavePlayer_0
nodes [{id=WavePlayer_0, properties {soundWaveAssetPtr=&/IP_PoppySoap_VO_Announcer_Fight_01}}]
properties {
 attenuationSettings=&/IP_ATT_PoppySoap_VO_Announcer
 maxDistance=18000
 soundClassObject=&Fort_UI_InGame_Critical
 virtualizationMode=Disabled
 volumeMultiplier=1.6
 concurrencySet [&/CONC_IP_PoppySoap_VO_Announcer]
}
