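# Translucent UI material for PoppySoap's "Fight" graphic, drawing text from a CoffeePuff SDF texture.
# Anti-aliasing functions threshold the SDF twice for text and wider outer-shadow masks, the latter scaled by a master opacity parameter.
# Vertically panning noise and a UV gradient mask accompany layered red-to-yellow lerps and a three-color blend that build the fiery color.
# A Crossfade_Flash parameter controls an additive brightening of the final color.
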Material /PoppySoap/UI/Assets/M_UI_PoppySoap_Fight
graph {
 properties {bAutomaticallySetUsageInEditor=false, blendMode=BLEND_Translucent, materialDomain=MD_UI, usedShadingModels=Default Lit}
 nodes [
  n1 = TextureSample() {samplerType=SAMPLERTYPE_Alpha, texture=&/T_UI_CoffeePuff_Text_SDF}
  n2 = Constant() {r=0.45}
  n3 = Add(A=@n2) {constB=-0.1}
  Multiply(
   A=NamedRerouteDeclaration(None=MaterialFunctionCall(DistanceField=@n1.RGB, Threshold=@n3) {materialFunction=&MF_UI_AntiAliasing}.Outer) {name=Mask_OuterShadow, nodeColor=<rgba 0.125 0.101 0.0625 1>}
   B=ScalarParameter() {defaultValue=1, parameterName=Config_MasterOpacity}
  )
  NamedRerouteDeclaration(None=MaterialFunctionCall(DistanceField=@n1.RGB, Threshold=@n2) {materialFunction=&MF_UI_AntiAliasing}.Outer) {name=Mask_Text, nodeColor=<rgba 0.125 0.101 0.0625 1>}
  NamedRerouteDeclaration(None=MaterialFunctionCall(Float2=TextureCoordinate()) {materialFunction=&BreakOutFloat2Components}.G) {name=Mask_LinearGradient, nodeColor=<rgba 0.125 0.101 0.0625 1>}
  n4 = TextureSample(UVs=Panner() {speedY=0.2}) {samplerType=SAMPLERTYPE_LinearColor, texture=&/T_UI_CoffeePuff_Text_Noise}
  n5 = Multiply(A=NamedRerouteUsage()) {constB=2}
  NamedRerouteDeclaration(None=Multiply(A=Multiply(A=@n4.R, B=@n5)) {constB=0.125}) {name=Texture_NoisePatternText, nodeColor=<rgba 0.125 0.101 0.0625 1>}
  n6 = Constant3Vector() {constant=<rgba 0.9473 0.0612 0.0222 1>}
  n7 = Constant3Vector() {constant=<rgba 1 1 0.4621 1>}
  n8 = NamedRerouteUsage()
  n9 = Multiply(A=@n8)
  n10 = NamedRerouteUsage()
  n11 = Saturate(None=Multiply(A=Add(A=@n10) {constB=-0.5}) {constB=4})
  n12 = Add(A=Multiply(A=Multiply(A=@n8, B=@n11)) {constB=3}, B=@n11)
  n13 = Power(Base=@n10) {constExponent=3}
  n14 = LinearInterpolate(
   A=LinearInterpolate(B=LinearInterpolate(A=@n6, B=@n7, Alpha=@n9), Alpha=NamedRerouteUsage())
   B=MaterialFunctionCall(
    A=Constant3Vector() {constant=<rgba 0.4452 0.0006 0.0006 1>}
    B=Constant3Vector() {constant=<rgba 1 0.5029 0 1>}
    C=@n7
    Alpha=Max(A=@n12, B=@n13)
   ) {materialFunction=&3ColorBlend}.Result
   Alpha=Multiply(A=NamedRerouteUsage()) {constB=0.75}
  )
  n15 = Multiply(A=@n14) {constB=0.1}
  n16 = ScalarParameter() {parameterName=Crossfade_Flash}
  NamedRerouteDeclaration(None=Add(A=@n14, B=LinearInterpolate(A=@n15, Alpha=@n16) {constB=0.5})) {name=Calc_FinalColor, nodeColor=<rgba 0.125 0.101 0.0625 1>}
  NamedRerouteUsage()
 ]
}
