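# Widget Blueprint that shows the PoppySoap announcer's "Fight" callout as a centered decal with a voice-over sound.
# On Construct it plays the sound and Anim_Intro, which drives the decal material's opacity and flash parameters.
# Playback speed is the reciprocal of DelayDuration, stretching the one-second animation across that time.
# After the same delay, it removes its own scene from the DynamicUIManager for the owning player.
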WidgetBlueprint /PoppySoap/UI/WBP_PoppySoap_AnnouncerVO_Fight
animations [Anim_Intro {
  binding_count=1
  length_seconds=1
  object=&self:Anim_Intro
  sequence {
   bindings [Image_FightDecal {
     id=#1
     possessed_class=Image
     tracks [{
       type=MovieSceneWidgetMaterialTrack
       sections [{
         type=MovieSceneParameterSection
         end_frame=30
         start_frame=0
         channels [
          Config_MasterOpacity {type=MovieSceneScriptingFloatChannel, keys [[0,0],[2,1],[25,1],[30,0]]}
          Crossfade_Flash {type=MovieSceneScriptingFloatChannel, keys [[0,1],[5,0]]}
         ]
        }]
      }]
    }]
   display_rate {numerator=30}
  }
 }]
graphs [EventGraph {nodes [
   Event Construct()
   > Parent: Construct()
   > PlaySound2D(Sound=/PoppySoap/Assets/Sound/IP_SC_PoppySoap_Announcer_VO_Fight)
   n1 = Get DelayDuration() -> (DelayDuration:float)
   > PlayAnimation(InAnimation=Get Anim_Intro() -> (Anim_Intro:Widget Animation Object Reference), PlaybackSpeed=float / float(A=1, B=@n1))
   > Delay(Duration=@n1)
   > RemoveScene!(
    execute
    self:Dynamic UIManager Object Reference=Get DynamicUIManager() -> (ReturnValue:Dynamic UIManager Object Reference) {is=title K2Node_GetSubsystem}
    Scene:Dynamic UIScene Object Reference=&/DynamicUIScene_PoppySoap_AnnouncerVO_Fight
    Player:"Player Controller Object Reference (by ref)"=GetOwningPlayer()
   ) -> (then)
  ]}]
variables [DelayDuration {type=number, default=4}]
widgets {
 info {inheritedWidgetCount=0, parentClass=&UserWidget, rootWidgetClass=&Overlay, widgetCount=2}
 widgets [
  Overlay_17 {widgetClassPath=&Overlay, widget=Overlay_17}
  Image_FightDecal {
   widgetClassPath=&Image
   bIsVariable=true
   parent=Overlay_17
   widget=Image_FightDecal
   properties {brush {
     bIsDynamicallyLoaded=false
     drawAs=Image
     imageType=FullColor
     mirroring=NoMirror
     resourceName=None
     resourceObject=/PoppySoap/UI/Assets/M_UI_PoppySoap_Fight
     tiling=NoTile
     imageSize=<300 150>
     margin=<ltrb 0 0 0 0>
     outlineSettings {
      bUseBrushTransparency=false
      roundingType=HalfHeightRadius
      width=0
      color {colorUseRule=UseColor_Specified, specifiedColor=<rgba 0 0 0 0>}
      cornerRadii=<0 0 0 0>
     }
     tintColor {colorUseRule=UseColor_Specified, specifiedColor=<rgba 1 1 1 1>}
     uVRegion {bIsValid=false, max=<0 0>, min=<0 0>}
    }}
   slot {class=OverlaySlot, properties {horizontalAlignment=HAlign_Center, verticalAlignment=VAlign_Center}}
  }
 ]
}
