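# Sound cue that plays the PoppySoap "Toasty" voice-over line through the critical in-game UI sound class.
# A single wave player feeds the output directly, with no randomization or other processing nodes, at a boosted volume.
# A concurrency override allows one instance at a time, with a new play stopping the older one.
# Virtualization is disabled and the max distance is set very large.
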SoundCue /PoppySoap/Assets/Sound/IP_SC_PoppySoap_Toasty_VO
root=WavePlayer_0
nodes [{id=WavePlayer_0, properties {soundWaveAssetPtr=&/IP_PoppySoap_VO_Toasty}}]
properties {
 bOverrideConcurrency=true
 maxDistance=2097152
 soundClassObject=&Fort_UI_InGame_Critical
 virtualizationMode=Disabled
 volumeMultiplier=1.5
 concurrencyOverrides {maxCount=1, resolutionRule=StopOldest}
}
