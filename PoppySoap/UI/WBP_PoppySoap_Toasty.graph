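# PoppySoap "Toasty" popup widget that slides an image in, plays a voice-over, then slides it out and removes itself.
# On Construct, it scales up for mobile touch players, starts the intro animation and plays the 2D sound.
# A Delay, wired from both the sound call and intro finish, leads to the outro animation.
# When the outro finishes, DynamicUIManager removes the Toasty scene for the owning player.
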WidgetBlueprint /PoppySoap/UI/WBP_PoppySoap_Toasty
animations [
 Anim_Toasty {
  binding_count=1
  length_seconds=0.75
  object=&self:Anim_Toasty
  sequence {bindings [Image_Toasty {
     id=#1
     possessed_class=Image
     tracks [{
       type=MovieScene2DTransformTrack
       sections [{
         type=MovieScene2DTransformSection
         end_frame=15
         start_frame=0
         channels [
          Translation.X {type=MovieSceneScriptingFloatChannel, keys [[0,256],[5,0],[5.0897,0]]}
          Translation.Y {type=MovieSceneScriptingFloatChannel, const [0,2]}
          Angle {type=MovieSceneScriptingFloatChannel, const [0,1]}
          Scale.X {type=MovieSceneScriptingFloatChannel, const [1,1]}
          Scale.Y {type=MovieSceneScriptingFloatChannel, const [1,1]}
          Shear.X {type=MovieSceneScriptingFloatChannel, const [0,1]}
          Shear.Y {type=MovieSceneScriptingFloatChannel, const [0,1]}
         ]
        }]
      }]
    }]}
 }
 Anim_Toasty_Outro {
  binding_count=1
  length_seconds=0.2522
  object=&self:Anim_Toasty_Outro
  sequence {bindings [Image_Toasty {
     id=#1
     possessed_class=Image
     tracks [{
       type=MovieScene2DTransformTrack
       sections [{
         type=MovieScene2DTransformSection
         end_frame=5
         start_frame=0
         channels [
          Translation.X {type=MovieSceneScriptingFloatChannel, keys [[0,0],[5.045,256]]}
          Translation.Y {type=MovieSceneScriptingFloatChannel, const [0,2]}
          Angle {type=MovieSceneScriptingFloatChannel, const [0,1]}
          Scale.X {type=MovieSceneScriptingFloatChannel, const [1,1]}
          Scale.Y {type=MovieSceneScriptingFloatChannel, const [1,1]}
          Shear.X {type=MovieSceneScriptingFloatChannel, const [0,1]}
          Shear.Y {type=MovieSceneScriptingFloatChannel, const [0,1]}
         ]
        }]
      }]
    }]}
 }
]
graphs [
 EventGraph {nodes [
   Event Construct()
   > SetMobileScale()
   > n1 = "Play Animation with Finished event (legacy)"!(
    execute:Exec
    Widget:User Widget Object Reference=Self-Reference()
    InAnimation:Widget Animation Object Reference=Get Anim_Toasty() -> (Anim_Toasty:Widget Animation Object Reference)
    StartAtTime:float=0
    NumLoopsToPlay:int=1
    PlayMode:EUMGSequencePlayMode Enum
    PlaybackSpeed:float=1
   ) -> (then:Exec, Result:UMGSequence Player Object Reference, Finished:Exec) {is=title K2Node_PlayAnimation}
   > n2 = PlaySound2D(Sound=/PoppySoap/Assets/Sound/IP_SC_PoppySoap_Toasty_VO)
   Delay(execute=[@n2.then, @n1.Finished], Duration=1.25)
   > n3 = "Play Animation with Finished event (legacy)"(
    Widget=Self-Reference()
    InAnimation=Get Anim_Toasty_Outro() -> (Anim_Toasty_Outro:Widget Animation Object Reference)
    StartAtTime=0
    NumLoopsToPlay=1
    PlaybackSpeed=1
   )
   RemoveScene!(
    execute=@n3.Finished
    self:Dynamic UIManager Object Reference=Get DynamicUIManager() -> (ReturnValue:Dynamic UIManager Object Reference) {is=title K2Node_GetSubsystem}
    Scene:Dynamic UIScene Object Reference=&/DynamicUIScene_PoppySoap_Toasty
    Player:"Player Controller Object Reference (by ref)"=GetOwningPlayer()
   ) -> (then)
  ]}
 SetMobileScale {nodes [
   SetMobileScale() {is=title K2Node_FunctionEntry}
   n1 = GetOwningPlayer()
   > Branch(Condition=AND Boolean(A=IsMobileGame(), B=IsUsingTouch(PlayerController=@n1)))
   > SetUserSpecifiedScale!(
    execute
    self:Scale Box Object Reference=Get ScaleBox_Mobile() -> (ScaleBox_Mobile:Scale Box Object Reference)
    InUserSpecifiedScale:float=1.5
   ) -> (then)
  ]}
]
variables []
widgets {
 info {inheritedWidgetCount=0, parentClass=&UserWidget, rootWidgetClass=&Overlay, widgetCount=3}
 widgets [
  Overlay_Root {widgetClassPath=&Overlay, widget=Overlay_Root}
  ScaleBox_Mobile {
   widgetClassPath=&ScaleBox
   bIsVariable=true
   parent=Overlay_Root
   widget=ScaleBox_Mobile
   properties {stretch=UserSpecified}
   slot {class=OverlaySlot, properties {horizontalAlignment=HAlign_Center, verticalAlignment=VAlign_Center}}
  }
  Image_Toasty {
   widgetClassPath=&Image
   bIsVariable=true
   parent=ScaleBox_Mobile
   widget=Image_Toasty
   properties {brush {
     bIsDynamicallyLoaded=false
     drawAs=Image
     imageType=FullColor
     mirroring=NoMirror
     resourceName=None
     resourceObject=/PoppySoap/UI/Assets/T_UI_PoppySoap_Toasty
     tiling=NoTile
     imageSize=<256 256>
     margin=<ltrb 0 0 0 0>
     outlineSettings {
      bUseBrushTransparency=false
      roundingType=HalfHeightRadius
      width=0
      color {colorUseRule=UseColor_Specified, specifiedColor=<rgba 0 0 0 0>}
      cornerRadii=<0 0 0 0>
     }
     tintColor {colorUseRule=UseColor_Specified, specifiedColor=<rgba 1 1 1 1>}
     uVRegion {bIsValid=false, max=<0 0>, min=<0 0>}
    }}
   slot {class=ScaleBoxSlot}
  }
 ]
}
